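# Physics asset for the dynamic skirt on the female medium Skirmish body, with pelvis, thigh and calf capsules as the body shapes.
# Five skirt chains use capsules at the back and boxes at the front, with limited swing and tight twist.
# Small MSTR root bodies attach to the pelvis, except front-right, which attaches to the right thigh.
# Angularly free cross-constraints tie back chains at segment 03 and front chains at 03 and 04.
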PhysicsAsset /Game/Characters/Player/Female/Medium/Bodies/F_MED_Skirmish/Meshes/F_MED_Skirmish_Dyn
bodies {
 calf_l [{shapeType=Capsule, length=38.1281, radius=7.9032, center=<-31.2496 -3.5293 -0.1472>, rotation=<rot 88.9719 -0.0001 5.9417>}]
 calf_r [{shapeType=Capsule, length=38.1281, radius=7.9032, center=<31.2496 3.5293 0.1472>, rotation=<rot 88.9719 -0.0001 5.9417>}]
 dyn_skirt_bk_01 [{shapeType=Capsule, length=6.0938, radius=6.25, center=<2.236 -0.0003 6>, rotation=<rot 90 0 0>}]
 dyn_skirt_bk_01_MSTR [{shapeType=Capsule, length=0.505, radius=0.505}]
 dyn_skirt_bk_02 [{shapeType=Capsule, length=6.0938, radius=6.25, center=<2.2361 -0.0003 6>, rotation=<rot 90 0 0>}]
 dyn_skirt_bk_03 [{shapeType=Capsule, length=6.0938, radius=6.25, center=<2.2361 -0.0003 6>, rotation=<rot 90 0 0>}]
 dyn_skirt_bk_l_01 [{shapeType=Capsule, length=6.3438, radius=6, center=<2.236 -0.0003 7.187>, rotation=<rot 90 0 0>}]
 dyn_skirt_bk_l_01_MSTR [{shapeType=Capsule, length=0.505, radius=0.505}]
 dyn_skirt_bk_l_02 [{shapeType=Capsule, length=6.3438, radius=6, center=<2.2361 -0.0003 6.7235>, rotation=<rot 90 0 0>}]
 dyn_skirt_bk_l_03 [{shapeType=Capsule, length=6.3438, radius=6, center=<2.2361 -0.0003 6>, rotation=<rot 90 0 0>}]
 dyn_skirt_bk_l_04 [{shapeType=Capsule, length=6.3438, radius=6, center=<2.2361 -0.0003 6>, rotation=<rot 90 0 0>}]
 dyn_skirt_bk_l_05 [{shapeType=Capsule, length=6.3438, radius=6, center=<2.2361 -0.0003 6>, rotation=<rot 90 0 0>}]
 dyn_skirt_bk_r_01 [{shapeType=Capsule, length=6.3438, radius=6, center=<2.236 -0.0003 7.187>, rotation=<rot 90 0 0>}]
 dyn_skirt_bk_r_01_MSTR [{shapeType=Capsule, length=0.505, radius=0.505}]
 dyn_skirt_bk_r_02 [{shapeType=Capsule, length=6.3438, radius=6, center=<2.2361 -0.0003 6>, rotation=<rot 90 0.0198 0.0198>}]
 dyn_skirt_bk_r_03 [{shapeType=Capsule, length=6.3438, radius=6, center=<2.2361 -0.0003 6>, rotation=<rot 90 0 0>}]
 dyn_skirt_bk_r_04 [{shapeType=Capsule, length=6.3438, radius=6, center=<2.2361 -0.0003 6>, rotation=<rot 90 0 0>}]
 dyn_skirt_bk_r_05 [{shapeType=Capsule, length=6.3438, radius=6, center=<2.2361 -0.0003 6>, rotation=<rot 90 0 0>}]
 dyn_skirt_fr_l_01 [{shapeType=Box, extentX=7.5, extentY=10.75, extentZ=1.8125, center=<3.9 3.3 0.5>}]
 dyn_skirt_fr_l_01_MSTR [{shapeType=Capsule, length=0.505, radius=0.505}]
 dyn_skirt_fr_l_02 [{shapeType=Box, extentX=9, extentY=10.75, extentZ=3, center=<3.9 3.3 0>}]
 dyn_skirt_fr_l_03 [{shapeType=Box, extentX=8.75, extentY=10.75, extentZ=3, center=<4.1905 3.3 0.0001>}]
 dyn_skirt_fr_l_04 [{shapeType=Box, extentX=3.875, extentY=10.75, extentZ=3, center=<1.6418 3.3 0>}]
 dyn_skirt_fr_r_01 [{shapeType=Box, extentX=7.5, extentY=10.75, extentZ=1.8125, center=<3.9 -3.3 0.5>}]
 dyn_skirt_fr_r_01_MSTR [{shapeType=Capsule, length=0.505, radius=0.505}]
 dyn_skirt_fr_r_02 [{shapeType=Box, extentX=9, extentY=10.75, extentZ=3, center=<3.9 -3.3 0>}]
 dyn_skirt_fr_r_03 [{shapeType=Box, extentX=8.75, extentY=10.75, extentZ=3, center=<4.1905 -3.3 0.0001>}]
 dyn_skirt_fr_r_04 [{shapeType=Box, extentX=3.875, extentY=10.75, extentZ=3, center=<1.6418 -3.3 0>}]
 pelvis [{shapeType=Capsule, length=13.7153, radius=11.0234, center=<-3.5389 -2.9678 0>}]
 thigh_l [
  {shapeType=Capsule, length=25.7531, radius=9.7782, center=<-19.1785 -1.7176 -0.3638>, rotation=<rot 88.97 0 3.2668>}
  {shapeType=Capsule, length=12.4375, radius=7.4062, center=<-40.7655 -2.3955 0.0001>, rotation=<rot 90 0 -16.0333>}
 ]
 thigh_r [
  {shapeType=Capsule, length=25.7531, radius=9.7782, center=<19.1785 1.7176 0.3638>, rotation=<rot 88.97 0 -176.7332>}
  {shapeType=Capsule, length=12.4375, radius=7.4062, center=<40.7655 2.3955 -0.0001>, rotation=<rot 90 0 163.9667>}
 ]
}
constraints [
 {bone1Name=dyn_skirt_bk_l_01, bone2Name=dyn_skirt_bk_l_01_MSTR, swing1LimitDegrees=15, swing2LimitDegrees=45, twistLimitDegrees=5}
 {bone1Name=dyn_skirt_bk_l_02, bone2Name=dyn_skirt_bk_l_01, swing1LimitDegrees=15, swing2LimitDegrees=45, twistLimitDegrees=5}
 {bone1Name=dyn_skirt_bk_l_03, bone2Name=dyn_skirt_bk_l_02, swing1LimitDegrees=15, swing2LimitDegrees=45, twistLimitDegrees=5}
 {bone1Name=dyn_skirt_bk_l_04, bone2Name=dyn_skirt_bk_l_03, swing1LimitDegrees=15, swing2LimitDegrees=45, twistLimitDegrees=5}
 {bone1Name=dyn_skirt_bk_l_05, bone2Name=dyn_skirt_bk_l_04, swing1LimitDegrees=15, swing2LimitDegrees=45, twistLimitDegrees=5}
 {bone1Name=dyn_skirt_bk_r_01, bone2Name=dyn_skirt_bk_r_01_MSTR, swing1LimitDegrees=15, swing2LimitDegrees=45, twistLimitDegrees=5}
 {bone1Name=dyn_skirt_bk_r_02, bone2Name=dyn_skirt_bk_r_01, swing1LimitDegrees=15, swing2LimitDegrees=45, twistLimitDegrees=5}
 {bone1Name=dyn_skirt_bk_r_03, bone2Name=dyn_skirt_bk_r_02, swing1LimitDegrees=15, swing2LimitDegrees=45, twistLimitDegrees=5}
 {bone1Name=dyn_skirt_bk_r_04, bone2Name=dyn_skirt_bk_r_03, swing1LimitDegrees=15, swing2LimitDegrees=45, twistLimitDegrees=5}
 {bone1Name=dyn_skirt_bk_r_05, bone2Name=dyn_skirt_bk_r_04, swing1LimitDegrees=15, swing2LimitDegrees=45, twistLimitDegrees=5}
 {bone1Name=dyn_skirt_fr_r_01, bone2Name=dyn_skirt_fr_r_01_MSTR, swing1LimitDegrees=15, swing2LimitDegrees=45, twistLimitDegrees=5}
 {bone1Name=dyn_skirt_fr_r_02, bone2Name=dyn_skirt_fr_r_01, swing1LimitDegrees=15, swing2LimitDegrees=45, twistLimitDegrees=5}
 {bone1Name=dyn_skirt_fr_r_03, bone2Name=dyn_skirt_fr_r_02, swing1LimitDegrees=25, swing2LimitDegrees=45, twistLimitDegrees=5}
 {bone1Name=dyn_skirt_fr_r_04, bone2Name=dyn_skirt_fr_r_03, swing1LimitDegrees=25, swing2LimitDegrees=45, twistLimitDegrees=5}
 {
  bone1Name=calf_r
  bone2Name=thigh_r
  swing1LimitDegrees=45
  swing1Motion=Free
  swing2LimitDegrees=45
  swing2Motion=Free
  twistLimitDegrees=45
  twistMotion=Free
 }
 {
  bone1Name=calf_l
  bone2Name=thigh_l
  swing1LimitDegrees=45
  swing1Motion=Free
  swing2LimitDegrees=45
  swing2Motion=Free
  twistLimitDegrees=45
  twistMotion=Free
 }
 {
  bone1Name=thigh_l
  bone2Name=pelvis
  swing1LimitDegrees=45
  swing1Motion=Free
  swing2LimitDegrees=45
  swing2Motion=Free
  twistLimitDegrees=45
  twistMotion=Free
 }
 {
  bone1Name=dyn_skirt_bk_l_01_MSTR
  bone2Name=pelvis
  swing1LimitDegrees=45
  swing1Motion=Free
  swing2LimitDegrees=45
  swing2Motion=Free
  twistLimitDegrees=45
  twistMotion=Free
 }
 {
  bone1Name=dyn_skirt_fr_l_01_MSTR
  bone2Name=pelvis
  swing1LimitDegrees=45
  swing1Motion=Free
  swing2LimitDegrees=45
  swing2Motion=Free
  twistLimitDegrees=45
  twistMotion=Free
 }
 {
  bone1Name=dyn_skirt_bk_r_01_MSTR
  bone2Name=pelvis
  swing1LimitDegrees=45
  swing1Motion=Free
  swing2LimitDegrees=45
  swing2Motion=Free
  twistLimitDegrees=45
  twistMotion=Free
 }
 {
  bone1Name=dyn_skirt_fr_r_01_MSTR
  bone2Name=thigh_r
  swing1LimitDegrees=45
  swing1Motion=Free
  swing2LimitDegrees=45
  swing2Motion=Free
  twistLimitDegrees=45
  twistMotion=Free
 }
 {
  bone1Name=dyn_skirt_bk_01_MSTR
  bone2Name=pelvis
  swing1LimitDegrees=45
  swing1Motion=Free
  swing2LimitDegrees=45
  swing2Motion=Free
  twistLimitDegrees=45
  twistMotion=Free
 }
 {bone1Name=dyn_skirt_bk_01, bone2Name=dyn_skirt_bk_01_MSTR, swing1LimitDegrees=15, swing2LimitDegrees=45, twistLimitDegrees=5}
 {bone1Name=dyn_skirt_bk_02, bone2Name=dyn_skirt_bk_01, swing1LimitDegrees=15, swing2LimitDegrees=45, twistLimitDegrees=5}
 {bone1Name=dyn_skirt_bk_03, bone2Name=dyn_skirt_bk_02, swing1LimitDegrees=15, swing2LimitDegrees=45, twistLimitDegrees=5}
 {
  bone1Name=dyn_skirt_bk_r_03
  bone2Name=dyn_skirt_bk_03
  swing1LimitDegrees=45
  swing1Motion=Free
  swing2LimitDegrees=45
  swing2Motion=Free
  twistLimitDegrees=45
  twistMotion=Free
 }
 {
  bone1Name=dyn_skirt_bk_l_03
  bone2Name=dyn_skirt_bk_03
  swing1LimitDegrees=45
  swing1Motion=Free
  swing2LimitDegrees=45
  swing2Motion=Free
  twistLimitDegrees=45
  twistMotion=Free
 }
 {bone1Name=dyn_skirt_fr_l_01, bone2Name=dyn_skirt_fr_l_01_MSTR, swing1LimitDegrees=22, swing2LimitDegrees=45, twistLimitDegrees=5}
 {bone1Name=dyn_skirt_fr_l_02, bone2Name=dyn_skirt_fr_l_01, swing1LimitDegrees=22, swing2LimitDegrees=45, twistLimitDegrees=5}
 {bone1Name=dyn_skirt_fr_l_03, bone2Name=dyn_skirt_fr_l_02, swing1LimitDegrees=25, swing2LimitDegrees=45, twistLimitDegrees=5}
 {bone1Name=dyn_skirt_fr_l_04, bone2Name=dyn_skirt_fr_l_03, swing1LimitDegrees=25, swing2LimitDegrees=45, twistLimitDegrees=5}
 {
  bone1Name=dyn_skirt_fr_r_03
  bone2Name=dyn_skirt_fr_l_03
  swing1LimitDegrees=45
  swing1Motion=Free
  swing2LimitDegrees=45
  swing2Motion=Free
  twistLimitDegrees=45
  twistMotion=Free
 }
 {
  bone1Name=dyn_skirt_fr_r_04
  bone2Name=dyn_skirt_fr_l_04
  swing1LimitDegrees=45
  swing1Motion=Free
  swing2LimitDegrees=45
  swing2Motion=Free
  twistLimitDegrees=45
  twistMotion=Free
 }
 {
  bone1Name=thigh_r
  bone2Name=pelvis
  swing1LimitDegrees=45
  swing1Motion=Free
  swing2LimitDegrees=45
  swing2Motion=Free
  twistLimitDegrees=45
  twistMotion=Free
 }
]
