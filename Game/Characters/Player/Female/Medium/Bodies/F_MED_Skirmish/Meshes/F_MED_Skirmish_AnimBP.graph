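# Post-processing AnimBlueprint for the F_MED_Skirmish female player body, adding corrective deformation and skirt dynamics.
# The parent mesh pose runs through Control Rigs and weighted front-skirt constraints into a RigidBody simulation.
# The event graph drives the simulation's gravity by state and rotates its base bone, dyn_root_simspace, in world space.
# Additive bone offsets adjust the front-left skirt during a custom emote and the back skirt outside the front end or while emoting.
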AnimBlueprint /Game/Characters/Player/Female/Medium/Bodies/F_MED_Skirmish/Meshes/F_MED_Skirmish_AnimBP
target_skeleton=&/F_MED_Skirmish_Skeleton
graphs [
 AnimGraph {nodes [
   n1 = Control Rig(Source=Copy Pose From Mesh() {bUseAttachedParent=true}) {
    alphaBoolBlend {blendOption=Linear}
    controlRigAssetReference {blueprintRigClass=&/F_MED_Skirmish_Deform_CtrlRig_twists.F_MED_Skirmish_Deform_CtrlRig_twists_C}
   }
   n2 = Control Rig!(
    Source=@n1
    Alpha
    KneeRot_fr_l_114deg:Vector
    KneeTrans_fr_l_114deg:Vector="0.258,-3.591,0.009"
    KneeScale_fr_l_114deg:Vector="1,1,1"
    KneeScale_fr_l_130deg:Vector="1,1,1"
    KneeRot_fr_l_130deg:Vector="0,0,-51.2"
    KneeTrans_fr_l_130deg:Vector="0.258,-3.591,0.009"
   ) -> (Pose) {
    alphaBoolBlend {blendOption=Linear}
    controlRigAssetReference {blueprintRigClass=&/F_MED_Skirmish_Deform_CtrlRig_sculpters.F_MED_Skirmish_Deform_CtrlRig_sculpters_C}
   }
   n3 = "Constraint - dyn_skirt_fr_l_01_MSTR (PPPP)"!(
    ComponentPose=Local To Component(LocalPose=@n2)
    Alpha=1
    ConstraintWeights_0=0.2623
    ConstraintWeights_1:float=0.4631
    ConstraintWeights_2:float=0.0326
    ConstraintWeights_3:float=0.242
   ) -> (Pose) {
    alphaBoolBlend {blendOption=Linear}
    boneToModify {boneName=dyn_skirt_fr_l_01_MSTR}
    constraintSetup [
     {targetBone {boneName=pelvis}}
     {targetBone {boneName=thigh_twist_01_l}}
     {targetBone {boneName=thigh_l}}
     {targetBone {boneName=deform_thigh_fr_l}}
    ]
    constraintWeights [1,1,1,1]
   }
   n4 = "Constraint - dyn_skirt_fr_r_01_MSTR (PPPP)"!(
    ComponentPose=@n3
    Alpha=1
    ConstraintWeights_0=0.0326
    ConstraintWeights_1:float=0.2623
    ConstraintWeights_2:float=0.4631
    ConstraintWeights_3:float=0.242
   ) -> (Pose) {
    alphaBoolBlend {blendOption=Linear}
    boneToModify {boneName=dyn_skirt_fr_r_01_MSTR}
    constraintSetup [
     {targetBone {boneName=thigh_r}}
     {targetBone {boneName=pelvis}}
     {targetBone {boneName=thigh_twist_01_r}}
     {targetBone {boneName=deform_thigh_fr_r}}
    ]
    constraintWeights [1,1,1,1]
   }
   n5 = Control Rig!(Source=Component To Local(ComponentPose=@n4), Alpha) -> (Pose) {
    alphaBoolBlend {blendOption=Linear}
    controlRigAssetReference {blueprintRigClass=&/F_MED_Skirmish_CtrlRig.F_MED_Skirmish_CtrlRig_C}
   }
   n6 = "Transform (Modify) Bone - Bone: dyn_skirt_fr_l_01_MSTR"(
    ComponentPose=Local To Component(LocalPose=@n5)
    Alpha=Get isCustomEmotePlaying() -> (isCustomEmotePlaying:double)
    Translation="0,2,2"
    Rotation="-50,0,0"
    Scale="1,1,1"
   ) {
    rotationMode=BMM_Additive
    rotationSpace=BCS_BoneSpace
    translationMode=BMM_Additive
    translationSpace=BCS_BoneSpace
    alphaBoolBlend {blendOption=Linear}
    alphaScaleBiasClamp {bInterpResult=true, interpSpeedDecreasing=2, interpSpeedIncreasing=2}
    boneToModify {boneName=dyn_skirt_fr_l_01_MSTR}
    rotation {pitch=-50}
    translation {y=2, z=2}
   }
   n7 = "Transform (Modify) Bone - Bone: dyn_root_simspace"(ComponentPose=@n6, Alpha=1, Translation="0,0,0", Rotation=Get sim_space_rotation() -> (sim_space_rotation:Rotator), Scale="1,1,1") {
    rotationMode=BMM_Replace
    rotationSpace=BCS_WorldSpace
    translationSpace=BCS_BoneSpace
    alphaBoolBlend {blendOption=Linear}
    alphaScaleBiasClamp {bInterpResult=true, interpSpeedDecreasing=2, interpSpeedIncreasing=2}
    boneToModify {boneName=dyn_root_simspace}
   }
   n8 = RigidBody!(ComponentPose=@n7, Alpha, OverrideWorldGravity:Vector=Get GravityOverride(), ExternalForce) -> (Pose) {
    bClampLinearTranslationLimitToRefPose=true
    bOverrideWorldGravity=true
    overridePhysicsAsset=&/F_MED_Skirmish_Dyn
    simulationSpace=BaseBoneSpace
    alphaBoolBlend {blendOption=Linear}
    baseBoneRef {boneName=dyn_root_simspace}
    componentAppliedLinearAccClamp=<3000 3000 3000>
    componentLinearAccScale {z=0.1}
    componentLinearVelScale {z=0.1}
    overrideWorldGravity {z=-980}
    simSpaceSettings {
     maxAngularAcceleration=1000
     maxAngularVelocity=3
     maxLinearAcceleration=700
     maxLinearVelocity=500
     velocityScaleZ=0
     worldAlpha=0.25
     externalLinearDragV=<0.2 0.2>
    }
   }
   n9 = Get NotFrontEnd() -> (NotFrontEnd:double)
   n10 = "Transform (Modify) Bone - Bone: dyn_skirt_bk_l_01"(ComponentPose=@n8, Alpha=@n9, Translation="0,0,4", Rotation="0,0,0", Scale="1,1,1") {
    rotationMode=BMM_Additive
    rotationSpace=BCS_BoneSpace
    translationMode=BMM_Additive
    translationSpace=BCS_BoneSpace
    alphaBoolBlend {blendOption=Linear}
    alphaScaleBiasClamp {bInterpResult=true, interpSpeedDecreasing=2, interpSpeedIncreasing=2}
    boneToModify {boneName=dyn_skirt_bk_l_01}
    translation {z=4}
   }
   n11 = "Transform (Modify) Bone - Bone: dyn_skirt_bk_r_01_MSTR"(ComponentPose=@n10, Alpha=@n9, Translation="0,0,4", Rotation="0,0,0", Scale="1,1,1") {
    rotationMode=BMM_Additive
    rotationSpace=BCS_BoneSpace
    translationMode=BMM_Additive
    translationSpace=BCS_BoneSpace
    alphaBoolBlend {blendOption=Linear}
    alphaScaleBiasClamp {bInterpResult=true, interpSpeedDecreasing=2, interpSpeedIncreasing=2}
    boneToModify {boneName=dyn_skirt_bk_r_01_MSTR}
    translation {z=4}
   }
   n12 = "Transform (Modify) Bone - Bone: dyn_skirt_bk_01"(ComponentPose=@n11, Alpha=@n9, Translation="0,0,4", Rotation="0,0,0", Scale="1,1,1") {
    rotationMode=BMM_Additive
    rotationSpace=BCS_BoneSpace
    translationMode=BMM_Additive
    translationSpace=BCS_BoneSpace
    alphaBoolBlend {blendOption=Linear}
    alphaScaleBiasClamp {bInterpResult=true, interpSpeedDecreasing=2, interpSpeedIncreasing=2}
    boneToModify {boneName=dyn_skirt_bk_01}
    translation {z=4}
   }
   Output Pose(Result=Component To Local(ComponentPose=@n12)) {layerGroup=None}
  ]}
 EventGraph {nodes [
   n1 = Event BlueprintUpdateAnimation()
   > n2 = Sequence!(execute) -> (then_0, then_1, then_2:Exec, then_3:Exec)
   n3 = GetAnimInstance(self=GetOwnerSkeletalMeshComponent())
   > Is Valid(InputObject=@n3)
   > Set isCustomEmotePlaying(isCustomEmotePlaying:double=SelectFloat(A=1, B=0, bPickA=Montage_IsPlaying(self=@n3, Montage=/Game/Animation/Game/MainPlayer/Emotes/Skirmish_CMF/Skirmish_CMF_M))) -> (Output_Get:double)
   n4 = "Gravity Override State Function (Parameter Asset) (Nativized)"(
    execute=@n2.then_1
    CharacterPartAnimInstance=Self-Reference()
    GameTimeInSeconds=GetGameTimeInSeconds()
    Parameters=&/F_MED_Skirmish_Parameters
   )
   > n5 = Set GravityOverride(GravityOverride=@n4)
   n6 = NOT Boolean(A=Get bIsInFrontEnd())
   n7 = Get bIsPlayingEmote()
   Set NotFrontEnd(execute=@n2.then_2, NotFrontEnd:double=SelectFloat(A=1, B=0, bPickA=OR Boolean(A=@n6, B=@n7))) -> (Output_Get:double)
   n8 = ComplexSimSpaceFunction DEPRECATED!(
    execute=@n2.then_3
    CharacterPartAnimInstance:Custom Character Part Anim Instance Object Reference=Self-Reference()
    InterpSpeedMax:double=9
    PawnSpeedMax:double=20
    RInterpErrorTolerance:double=1
    DeltaTime:double=@n1.DeltaTimeX
    DebugDraw:bool=false
    sim_space_rotation_in:Rotator=Get sim_space_rotation()
    inComponentSpaceIn:bool=Get InComponentSpace() -> (InComponentSpace:bool)
    inComponentSpaceLastFrameIn:bool=Get InComponentSpaceLastFrame() -> (InComponentSpaceLastFrame:bool)
    lockedOnSimSpaceIn:bool=Get LockedOnSimSpace() -> (LockedOnSimSpace:bool)
    InterpSpeedIn:double=Get InterpSpeed() -> (InterpSpeed:double)
   ) -> (
    then
    sim_space_rotation:Rotator
    inComponentSpaceOut:bool
    inComponentSpaceLastFrameOut:bool
    lockedOnSimSpaceOut:bool
    InterpSpeedOut:double
   )
   > Set sim_space_rotation(sim_space_rotation:Rotator=@n8) -> (Output_Get:Rotator)
   > Set InComponentSpace(InComponentSpace:bool=@n8.inComponentSpaceOut) -> (Output_Get:bool)
   > Set InComponentSpaceLastFrame(InComponentSpaceLastFrame:bool=@n8.inComponentSpaceLastFrameOut) -> (Output_Get:bool)
   > Set LockedOnSimSpace(LockedOnSimSpace:bool=@n8.lockedOnSimSpaceOut) -> (Output_Get:bool)
   > Set InterpSpeed(InterpSpeed:double=@n8.InterpSpeedOut) -> (Output_Get:double)
   Event OnEndCrouching()
   > Set crouch alpha(crouch alpha:double=0) -> (Output_Get:double)
   Event OnBeginCrouching()
   > Set crouch alpha(crouch alpha=1)
   DrawVectorFromJointFunction!(
    execute
    SkelMeshComponent:Scene Component Object Reference
    JointName:Name=pelvis
    LineEnd:Vector=@n5
    JointOffset:Vector="0, 0, 0"
    LineColor:Linear Color Structure=(R=1,G=0,B=0,A=1)
    Thickness:double=2
    Foreground:EDrawDebugSceneDepthPriorityGroup Enum=Foreground
   ) -> (then)
  ]}
]
variables [
 crouch alpha {type=number}
 Gravity Override Params {
  type=GravityOverrideParamsStruct
  default {
   emoteOrMelee {gravityOverride=<0 -50 -15>, windAmplitude=<50 25 50>, windFrequency=<0.1 0.2 0.15>, windOffset=<0 -50 -50>}
   front_End {gravityOverride=<0 -50 -15>, windAmplitude=<50 25 50>, windFrequency=<0.1 0.2 0.15>, windOffset=<0 -50 -50>}
   onGround_Moving {gravityOverride=<0 -50 -15>, windAmplitude=<50 25 50>, windFrequency=<0.1 0.2 0.15>, windOffset=<0 -50 -50>}
   onGround_Standing {gravityOverride=<0 -50 -15>, windAmplitude=<50 25 50>, windFrequency=<0.1 0.2 0.15>, windOffset=<0 -50 -50>}
   skydive_Parachuting {gravityOverride=<0 -300 -300>}
   swimming {windAmplitude=<50 -25 100>, windFrequency=<0.15 1 0.3>, windOffset=<0 -50 -50>}
  }
 }
 GravityOverride {type=Vector}
 HairRotator_01 {type=Rotator}
 HairRotator_02 {type=Rotator}
 HairRotator_03 {type=Rotator}
 HairRotator_04 {type=Rotator}
 HairVector_01 {type=Vector}
 HairVector_02 {type=Vector}
 HairVector_03 {type=Vector}
 HairVector_04 {type=Vector}
 InComponentSpace {type=boolean}
 InComponentSpaceLastFrame {type=boolean}
 InterpSpeed {type=number}
 isCustomEmotePlaying {type=number}
 LockedOnSimSpace {type=boolean}
 NotFrontEnd {type=number}
 sim_space_rotation {type=Rotator}
]
